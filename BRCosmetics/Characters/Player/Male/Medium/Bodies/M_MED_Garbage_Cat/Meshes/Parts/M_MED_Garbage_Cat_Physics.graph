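# Physics asset for the Garbage Cat medium male player body, setting up jiggle bones for its mustache and a tube.
# Left and right four-joint mustache capsule chains hang off the head, with twist and one swing axis locked, leaving only slight bend.
# A three-joint tube chain hangs off spine_05 with wider angular limits.
# Larger capsules on head and spine_05 act as chain roots, joined by an angularly free constraint.
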PhysicsAsset /BRCosmetics/Characters/Player/Male/Medium/Bodies/M_MED_Garbage_Cat/Meshes/Parts/M_MED_Garbage_Cat_Physics
bodies {
 dyn_L_mustache_jnt_1 [{shapeType=Capsule, length=0.3, radius=0.8, center=<0.1612 0 -0.0505>, rotation=<rot 72.5961 -0.0001 -0.0001>}]
 dyn_L_mustache_jnt_2 [{shapeType=Capsule, length=0.3, radius=0.8, center=<0.047 0 -0.0147>, rotation=<rot 72.5961 0 0>}]
 dyn_L_mustache_jnt_3 [{shapeType=Capsule, length=0.3, radius=0.8, center=<0.4475 0 -0.1403>, rotation=<rot 72.5961 0 0>}]
 dyn_L_mustache_jnt_4 [{shapeType=Capsule, length=0.3, radius=0.8, center=<0.1779 0 -0.0558>, rotation=<rot 72.5959 0 0>}]
 dyn_R_mustache_jnt_1 [{shapeType=Capsule, length=0.3, radius=0.8, center=<0.1612 0 -0.0505>, rotation=<rot 72.5961 0.0001 0>}]
 dyn_R_mustache_jnt_2 [{shapeType=Capsule, length=0.3, radius=0.8, center=<0.047 0 -0.0147>, rotation=<rot 72.5961 -0.0001 -0.0001>}]
 dyn_R_mustache_jnt_3 [{shapeType=Capsule, length=0.3, radius=0.8, center=<0.4475 0 -0.1403>, rotation=<rot 72.5961 0 0>}]
 dyn_R_mustache_jnt_4 [{shapeType=Capsule, length=0.3, radius=0.8, center=<0.4475 0 -0.1403>, rotation=<rot 72.596 0 0>}]
 dyn_tube_jnt_1 [{shapeType=Capsule, length=2, radius=1.5, center=<1.4199 0 -0.0169>, rotation=<rot -88.9019 179.9995 -179.9995>}]
 dyn_tube_jnt_2 [{shapeType=Capsule, length=2, radius=1.5, center=<1.4199 0 -0.0167>, rotation=<rot -88.9146 179.9996 -179.9995>}]
 dyn_tube_jnt_3 [{shapeType=Capsule, length=2, radius=1.5, center=<1.4199 0 -0.0159>, rotation=<rot -88.9724 179.9998 -179.9998>}]
 head [{shapeType=Capsule, length=5, radius=7, center=<3.6361 2.0642 0>}]
 spine_05 [{shapeType=Capsule, length=10, radius=6, center=<8.5983 0.7457 -2.6986>, rotation=<rot 30.2602 114.8263 -86.5851>}]
}
constraints [
 {
  bone1Name=dyn_R_mustache_jnt_2
  bone2Name=dyn_R_mustache_jnt_1
  swing1LimitDegrees=3
  swing1Motion=Locked
  swing2LimitDegrees=3
  twistLimitDegrees=1
  twistMotion=Locked
 }
 {
  bone1Name=dyn_R_mustache_jnt_3
  bone2Name=dyn_R_mustache_jnt_2
  swing1LimitDegrees=3
  swing1Motion=Locked
  swing2LimitDegrees=3
  twistLimitDegrees=1
  twistMotion=Locked
 }
 {
  bone1Name=dyn_R_mustache_jnt_4
  bone2Name=dyn_R_mustache_jnt_3
  swing1LimitDegrees=3
  swing1Motion=Locked
  swing2LimitDegrees=3
  twistLimitDegrees=1
  twistMotion=Locked
 }
 {
  bone1Name=dyn_L_mustache_jnt_2
  bone2Name=dyn_L_mustache_jnt_1
  swing1LimitDegrees=3
  swing1Motion=Locked
  swing2LimitDegrees=3
  twistLimitDegrees=1
  twistMotion=Locked
 }
 {
  bone1Name=dyn_L_mustache_jnt_3
  bone2Name=dyn_L_mustache_jnt_2
  swing1LimitDegrees=3
  swing1Motion=Locked
  swing2LimitDegrees=3
  twistLimitDegrees=1
  twistMotion=Locked
 }
 {
  bone1Name=dyn_L_mustache_jnt_4
  bone2Name=dyn_L_mustache_jnt_3
  swing1LimitDegrees=3
  swing1Motion=Locked
  swing2LimitDegrees=3
  twistLimitDegrees=1
  twistMotion=Locked
 }
 {bone1Name=dyn_tube_jnt_2, bone2Name=dyn_tube_jnt_1, swing1LimitDegrees=40, swing2LimitDegrees=40, twistLimitDegrees=10}
 {bone1Name=dyn_tube_jnt_3, bone2Name=dyn_tube_jnt_2, swing1LimitDegrees=40, swing2LimitDegrees=40, twistLimitDegrees=10}
 {
  bone1Name=dyn_R_mustache_jnt_1
  bone2Name=head
  swing1LimitDegrees=3
  swing1Motion=Locked
  swing2LimitDegrees=3
  twistLimitDegrees=1
  twistMotion=Locked
 }
 {
  bone1Name=dyn_L_mustache_jnt_1
  bone2Name=head
  swing1LimitDegrees=3
  swing1Motion=Locked
  swing2LimitDegrees=3
  twistLimitDegrees=1
  twistMotion=Locked
 }
 {bone1Name=dyn_tube_jnt_1, bone2Name=spine_05, swing1LimitDegrees=30, swing2LimitDegrees=15, twistLimitDegrees=10}
 {
  bone1Name=head
  bone2Name=spine_05
  swing1LimitDegrees=45
  swing1Motion=Free
  swing2LimitDegrees=45
  swing2Motion=Free
  twistLimitDegrees=45
  twistMotion=Free
 }
]
