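# Post-process-style cosmetic AnimBP for the M_MED_Garbage_Cat player body part, adding secondary motion to its tube and scarf bones.
# It copies the attached parent mesh's pose, runs RigidBody physics relative to spine_04, then additively rolls dyn_tube_jnt_1.
# Anim Dynamics on dyn_scarf_placeholder_jnt follows, selecting gravity scale by bIsFalling, then feeds the output.
# A three-node tubeScarf Anim Dynamics chain is unconnected, and the update event is empty.
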AnimBlueprint /BRCosmetics/Characters/Player/Male/Medium/Bodies/M_MED_Garbage_Cat/Meshes/Parts/M_MED_Garbage_Cat_AnimBP
target_skeleton=&/M_MED_Garbage_Cat_Skeleton
graphs [
 AnimGraph {nodes [
   n1 = Anim Dynamics - Bone: dyn_tubeScarf_R_jnt(Alpha=1, ExternalForce="0,0,0") {
    angularSpringConstant=100
    bGravityOverrideInSimSpace=true
    bOverrideLinearDamping=true
    linearDampingOverride=1
    linearSpringConstant=10
    simulationSpace=Actor
    alphaBoolBlend {blendOption=Linear}
    boundBone {boneName=dyn_tubeScarf_R_jnt}
    componentAppliedLinearAccClamp=<7000 7000 7000>
    componentLinearAccScale=<0.2 0.2 0.2>
    componentLinearVelScale=<0.4 0.4 0.4>
    physicsBodyDefinitions [{
      sphereCollisionRadius=0
      boundBone {boneName=dyn_tubeScarf_R_jnt}
      boxExtents=<1 1 -1>
      constraintSetup {angularLimitsMax=<5 5 5>, angularLimitsMin=<-5 -5 -5>, angularTarget=<0 0 0>, linearAxesMax=<0.2 1 0>, linearAxesMin=<-0.2 -1 0>}
      localJointOffset=<-0.2 -0.2 -0.2>
     }]
    relativeSpaceBone {boneName=root}
   }
   n2 = Anim Dynamics - Bone: dyn_tubeScarf_L_jnt(ComponentPose=@n1, Alpha=1, ExternalForce="0,0,0") {
    angularSpringConstant=100
    bGravityOverrideInSimSpace=true
    bOverrideLinearDamping=true
    linearDampingOverride=1
    linearSpringConstant=10
    simulationSpace=Actor
    alphaBoolBlend {blendOption=Linear}
    boundBone {boneName=dyn_tubeScarf_L_jnt}
    componentAppliedLinearAccClamp=<7000 7000 7000>
    componentLinearAccScale=<0.2 0.2 0.2>
    componentLinearVelScale=<0.4 0.4 0.4>
    physicsBodyDefinitions [{
      sphereCollisionRadius=0
      boundBone {boneName=dyn_tubeScarf_L_jnt}
      boxExtents=<1 1 -1>
      constraintSetup {angularLimitsMax=<5 5 5>, angularLimitsMin=<-5 -5 -5>, angularTarget=<0 0 0>, linearAxesMax=<0.2 1 0>, linearAxesMin=<-0.2 -1 0>}
      localJointOffset=<-0.2 -0.2 -0.2>
     }]
    relativeSpaceBone {boneName=root}
   }
   Anim Dynamics - Bone: dyn_tubeScarf_back_jnt(ComponentPose=@n2, Alpha=1, ExternalForce="0,0,0") {
    angularSpringConstant=100
    bGravityOverrideInSimSpace=true
    bOverrideLinearDamping=true
    linearDampingOverride=1
    linearSpringConstant=10
    simulationSpace=Actor
    alphaBoolBlend {blendOption=Linear}
    boundBone {boneName=dyn_tubeScarf_back_jnt}
    componentAppliedLinearAccClamp=<7000 7000 7000>
    componentLinearAccScale=<0.2 0.2 0.2>
    componentLinearVelScale=<0.4 0.4 0.4>
    physicsBodyDefinitions [{
      sphereCollisionRadius=0
      boundBone {boneName=dyn_tubeScarf_back_jnt}
      boxExtents=<1 1 -1>
      constraintSetup {angularLimitsMax=<5 5 5>, angularLimitsMin=<-5 -5 -5>, angularTarget=<0 0 0>, linearAxesMax=<0.5 0.6 0>, linearAxesMin=<-0.5 -0.6 0>}
      localJointOffset=<-0.2 -0.2 -0.2>
     }]
    relativeSpaceBone {boneName=root}
   }
   n3 = RigidBody(ComponentPose=Local To Component(LocalPose=Copy Pose From Mesh() {bUseAttachedParent=true})) {
    bClampLinearTranslationLimitToRefPose=true
    overridePhysicsAsset=&/M_MED_Garbage_Cat_Physics
    simulationSpace=BaseBoneSpace
    alphaBoolBlend {blendOption=Linear}
    baseBoneRef {boneName=spine_04}
    componentAppliedLinearAccClamp=<6000 6000 6000>
    componentLinearAccScale=<0.25 0.25 0.25>
    componentLinearVelScale=<0.3 0.3 0.3>
   }
   n4 = "Transform (Modify) Bone - Bone: dyn_tube_jnt_1"(ComponentPose=@n3, Alpha=1, Translation="0,0,0", Rotation="0,0,-10", Scale="1,1,1") {rotationMode=BMM_Additive, alphaBoolBlend {blendOption=Linear}, boneToModify {boneName=dyn_tube_jnt_1}, rotation {roll=-10}}
   n5 = Anim Dynamics - Bone: dyn_scarf_placeholder_jnt!(ComponentPose=@n4, Alpha=1, GravityScale:float=SelectFloat(A=0.3, B=0.25, bPickA=Get bIsFalling()), ExternalForce="0,0,0") -> (Pose) {
    angularSpringConstant=100
    bGravityOverrideInSimSpace=true
    bOverrideLinearDamping=true
    gravityScale=0.25
    linearDampingOverride=0.9
    linearSpringConstant=300
    simulationSpace=RootRelative
    alphaBoolBlend {blendOption=Linear}
    boundBone {boneName=dyn_scarf_placeholder_jnt}
    componentAppliedLinearAccClamp=<10000 10000 500>
    componentLinearAccScale {z=0.1}
    componentLinearVelScale=<0.7 0.7 0.7>
    physicsBodyDefinitions [{
      sphereCollisionRadius=0
      boundBone {boneName=dyn_scarf_placeholder_jnt}
      boxExtents=<5 1 1>
      constraintSetup {angularLimitsMax=<0 0 0>, angularLimitsMin=<0 0 0>, angularTarget=<0 0 0>, linearAxesMax=<0 0 2.7>}
      localJointOffset=<0 0 0>
     }]
    relativeSpaceBone {boneName=pelvis}
   }
   Output Pose(Result=Component To Local(ComponentPose=@n5)) {layerGroup=None}
  ]}
 EventGraph {nodes [Event BlueprintUpdateAnimation()]}
]
variables []
